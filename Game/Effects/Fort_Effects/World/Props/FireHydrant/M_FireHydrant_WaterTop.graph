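# Translucent lit mesh-particle material for the top of the water effect on the fire hydrant prop.
# A fast vertically panning tiled texture, partly desaturated and added to a dark blue tint, is multiplied by particle color as diffuse through ConvertFromDiffSpec.
# A water-top ramp mask times particle alpha is clamped and depth-faded to soften intersections.
# The same ramp also feeds a separate scaled multiply.
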Material /Game/Effects/Fort_Effects/World/Props/FireHydrant/M_FireHydrant_WaterTop
graph {
 properties {
  bUsedWithMeshParticles=true
  blendMode=BLEND_Translucent
  pixelDepthOffsetMode=PDOM_Legacy
  translucencyLightingMode=TLM_VolumetricPerVertexNonDirectional
  usedShadingModels=Default Lit
 }
 nodes [
  n1 = TextureSample() {texture=/Game/Global/Textures/Default/Masks_and_Gradients/T_Water_Top_Ramp}
  n2 = ParticleColor()
  DepthFade(Opacity=Clamp(None=Multiply(A=@n1.RGB, B=@n2.A))) {fadeDistanceDefault=15}
  n3 = Multiply(A=Constant() {r=1}, B=Constant3Vector() {constant=<rgba 0.02 0.05 0.1 0>})
  n4 = TextureCoordinate() {uTiling=2, vTiling=2}
  n5 = Desaturation(
   None=TextureSample(UVs=Panner(Coordinate=@n4) {speedX=0.05, speedY=3}) {texture=/Game/Effects/Fort_Effects/Textures/T_Tile_1}.RGB
   Fraction=Constant() {r=0.7}
  )
  MaterialFunctionCall(DiffuseColor=Multiply(A=Add(A=@n3, B=@n5), B=@n2.RGB)) {materialFunction=&ConvertFromDiffSpec}
  Multiply(A=@n1.RGB, B=Constant() {r=25})
 ]
}
