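# Translucent Default Lit sprite-particle material for a water splash, stored with the fire hydrant prop effects.
# A splash texture is desaturated by a constant fraction, tinted by particle color, and fed as diffuse color into ConvertFromDiffSpec.
# Texture alpha times particle alpha feeds a depth fade to soften intersections with scene geometry.
# Lighting uses the volumetric per-vertex non-directional translucency mode.
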Material /Game/Effects/Fort_Effects/World/Props/FireHydrant/M_WaterSplash_2
graph {
 properties {
  bUsedWithParticleSprites=true
  blendMode=BLEND_Translucent
  pixelDepthOffsetMode=PDOM_Legacy
  translucencyLightingMode=TLM_VolumetricPerVertexNonDirectional
  usedShadingModels=Default Lit
 }
 nodes [
  Constant()
  n1 = TextureSample() {texture=/Game/Effects/Fort_Effects/Textures/Water/T_Water_Splash}
  n2 = ParticleColor()
  DepthFade(Opacity=Multiply(A=@n1.A, B=@n2.A)) {fadeDistanceDefault=30}
  n3 = Constant() {r=0.8}
  MaterialFunctionCall(DiffuseColor=Multiply(A=@n2.RGB, B=Desaturation(None=@n1.RGB, Fraction=@n3))) {materialFunction=&ConvertFromDiffSpec}
 ]
}
