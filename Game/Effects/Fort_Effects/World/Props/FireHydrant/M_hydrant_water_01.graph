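# Translucent lit mesh-particle material for fire hydrant water, with optional refraction.
# A panning distortion map offsets two scrolling sea-foam layers that supply color, opacity and refraction normals at shared adjustable scale and speed.
# A flippable vertical gradient and particle alpha drive a power-based opacity erosion, while particle color tints diffuse and emissive.
# Parameters toggle Fresnel opacity and horizontal edge fading, and depth fade softens intersections.
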Material /Game/Effects/Fort_Effects/World/Props/FireHydrant/M_hydrant_water_01
graph {
 properties {
  bTangentSpaceNormal=false
  bUsedWithMeshParticles=true
  blendMode=BLEND_Translucent
  numCustomizedUVs=1
  opacityMaskClipValue=0.999
  pixelDepthOffsetMode=PDOM_Legacy
  refractionMethod=RM_IndexOfRefraction
  translucencyLightingMode=TLM_VolumetricPerVertexNonDirectional
  translucentShadowDensityScale=0
  usedShadingModels=Default Lit
 }
 nodes [
  n1 = TextureCoordinate()
  n2 = VectorParameter() {parameterName=UV Scale And Animation Speed, defaultValue=<rgba 1 1 1 1>}
  n3 = MaterialFunctionCall(R=@n2.R, G=@n2.R, B=@n2.R, A=@n2.R) {materialFunction=&AppendMany}
  n4 = ComponentMask(None=@n3.RGBA) {g=true, r=true}
  n5 = Constant2Vector() {g=1}
  n6 = ComponentMask(None=@n3.RGBA) {a=true, b=true}
  n7 = TextureSample(UVs=Panner(Coordinate=Multiply(A=@n1, B=@n4), Speed=Multiply(A=@n5, B=@n6)) {speedY=1}) {texture=/Game/Global/Textures/Default/Tileable_Patterns/T_Tilling_Distortion_Map_01}
  n8 = Multiply(A=ComponentMask(None=@n7.G) {g=true, r=true}) {constB=0.1}
  n9 = TextureCoordinate()
  n10 = Multiply(A=Constant2Vector() {g=1.5, r=1}, B=@n4)
  n11 = Panner(Coordinate=Add(A=@n8, B=Multiply(A=@n9, B=@n10)), Speed=Multiply(A=Constant2Vector() {g=1.6, r=-0.5}, B=@n6)) {speedX=-0.5, speedY=1.6}
  n12 = TextureSample(UVs=@n11) {texture=/Game/Packages/Fortress_Transylvania/Haunted_Forest/Swamps/Textures/T_sea_foam_03}
  n13 = TextureCoordinate()
  n14 = Multiply(A=Constant2Vector() {g=1.5, r=1}, B=@n4)
  n15 = Panner(Coordinate=Add(A=@n8, B=Multiply(A=@n13, B=@n14)), Speed=Multiply(A=Constant2Vector() {g=1.2, r=0.4}, B=@n6)) {speedX=0.4, speedY=1.2}
  n16 = TextureSample(UVs=@n15) {texture=&T_sea_foam_03}
  n17 = Multiply(A=@n16.A) {constB=0.4}
  n18 = MaterialFunctionCall() {materialFunction=&LinearGradient}
  n19 = LinearInterpolate(A=@n18.VGradient, B=OneMinus(None=@n18.VGradient), Alpha=ScalarParameter() {parameterName=FlipGradient})
  n20 = Clamp(None=Multiply(A=Power(Base=@n19) {constExponent=20}) {constB=2})
  n21 = Add(A=Add(A=@n12.G, B=Multiply(A=@n16.G) {constB=0.4}), B=@n20)
  n22 = ParticleColor()
  n23 = Clamp(None=Multiply(A=Power(Base=@n19) {constExponent=3}) {constB=5})
  n24 = Multiply(A=@n22.A, B=Clamp(None=@n23))
  n25 = LinearInterpolate(Alpha=@n24) {constA=5, constAlpha=1}
  n26 = Multiply(A=Multiply(A=Power(Base=@n21, Exp=@n25) {constExponent=500}, B=LinearInterpolate(Alpha=@n24) {constAlpha=0, constB=10}) {constB=5}) {constB=10}
  n27 = Multiply(A=@n24) {constB=0.45}
  n28 = Multiply(A=Add(A=Add(A=@n12.A, B=@n17), B=@n20), B=Clamp(None=Add(A=@n26, B=@n27)))
  n29 = Clamp(None=MaterialFunctionCall(Use Cheap contrast=StaticBool() {value=true}) {materialFunction=&Fresnel_Function}.Result)
  n30 = Clamp(None=Multiply(A=Power(Base=@n29)) {constB=5})
  n31 = DepthFade(
   Opacity=LinearInterpolate(A=@n28, B=LinearInterpolate(A=@n28, B=@n24, Alpha=@n30) {constA=1}, Alpha=ScalarParameter() {defaultValue=1, parameterName=Use Fresenel})
   FadeDistance=ScalarParameter() {defaultValue=30, parameterName=Depth Fade Distance}
  ) {fadeDistanceDefault=30}
  n32 = ComponentMask(None=TextureCoordinate()) {r=true}
  n33 = Multiply(A=@n32, B=OneMinus(None=@n32))
  n34 = ScalarParameter() {parameterName=Edge Fade Power}
  n35 = Clamp(None=Multiply(A=Power(Base=@n33, Exp=@n34), B=ScalarParameter() {parameterName=Edge Fade Multiplier}))
  LinearInterpolate(A=@n31, B=Multiply(A=@n31, B=@n35), Alpha=ScalarParameter() {parameterName=Fade Edges})
  n36 = Constant3Vector() {constant=<rgba 0 0 1 0>}
  n37 = LinearInterpolate(
   A=@n36
   B=Add(
    A=TextureSample(UVs=@n11) {samplerType=SAMPLERTYPE_Normal, texture=/Game/Packages/Fortress_Transylvania/Haunted_Forest/Swamps/Textures/T_sea_foam_03_n}.RGB
    B=TextureSample(UVs=@n15) {samplerType=SAMPLERTYPE_Normal, texture=&T_sea_foam_03_n}.RGB
   )
   Alpha=@n22.A
  )
  n38 = OneMinus(None=@n23)
  n39 = Clamp(None=Power(Base=Clamp(None=@n38), Exp=LinearInterpolate(Alpha=@n22.A) {constB=200}) {constExponent=35})
  n40 = Multiply(A=Clamp(None=LinearInterpolate(A=@n37, B=@n36, Alpha=@n39) {constA=1})) {constB=0.09}
  n41 = Multiply(A=@n40, B=@n35)
  n42 = ScalarParameter() {parameterName=Fade Edges}
  LinearInterpolate(
   A=Constant() {r=1}
   B=Add(A=LinearInterpolate(A=@n40, B=@n41, Alpha=@n42), B=Constant() {r=1})
   Alpha=ScalarParameter() {defaultValue=1, parameterName=Use Refraction}
  )
  n43 = Clamp(None=LinearInterpolate(A=Add(A=@n12.G, B=@n16.G), B=@n12.G, Alpha=@n12.G) {constA=5, constAlpha=1})
  n44 = Add(A=@n7.G, B=@n20)
  n45 = Multiply(A=@n22.RGB, B=Multiply(A=Add(A=@n43, B=@n44)) {constB=0.5})
  MaterialFunctionCall(DiffuseColor=@n45) {materialFunction=&ConvertFromDiffSpec}
  MaterialFunctionCall(Emissive=Multiply(A=@n45, B=VectorParameter() {parameterName=Emissive Multiplier, defaultValue=<rgba 0 0 0 1>}.RGB)) {materialFunction=&MF_FX_EmissiveAdjustments}
 ]
}
