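# Translucent, lit particle sprite material for water drops in the fire hydrant effect.
# A slowly panning infected-map pattern offsets the UVs of the drop texture to distort its shape.
# The drop's red channel, tinted by particle color, feeds a diffuse-to-PBR conversion, and its square is multiplied by particle alpha.
# That squared mask also blends a flat normal toward an intensified panning normal map.
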Material /Game/Effects/Fort_Effects/World/Props/FireHydrant/M_water_drop_01
graph {
 properties {
  bUsedWithParticleSprites=true
  blendMode=BLEND_Translucent
  opacityMaskClipValue=0.933
  pixelDepthOffsetMode=PDOM_Legacy
  refractionMethod=RM_IndexOfRefraction
  translucencyLightingMode=TLM_VolumetricPerVertexNonDirectional
  usedShadingModels=Default Lit
 }
 nodes [
  n1 = Constant3Vector() {constant=<rgba 0 0 1 0>}
  n2 = Panner(Coordinate=TextureCoordinate() {uTiling=0.15, vTiling=0.15}) {speedY=-0.16}
  n3 = Multiply(
   A=TextureSample(UVs=@n2) {samplerType=SAMPLERTYPE_Normal, texture=/Game/Global/Textures/Default/Tileable_Patterns/2ndInfectedMapNormal}.RGB
   B=Constant3Vector() {constant=<rgba 5 5 1 0>}
  )
  n4 = ComponentMask(None=TextureSample(UVs=@n2) {texture=&2ndInfectedMapLinear}.R) {g=true, r=true}
  n5 = TextureSample(UVs=Add(A=Multiply(A=@n4) {constB=0.2}, B=TextureCoordinate())) {texture=&/T_water_drop_01}
  n6 = Multiply(A=@n5.R, B=@n5.R)
  n7 = Clamp(None=@n6)
  Add(A=Clamp(None=LinearInterpolate(A=@n1, B=@n3, Alpha=@n7)), B=Constant() {r=1})
  n8 = ParticleColor()
  MaterialFunctionCall(DiffuseColor=Multiply(A=@n5.R, B=@n8.RGB)) {materialFunction=&ConvertFromDiffSpec}
  Multiply(A=Multiply(A=@n6), B=@n8.A)
 ]
}
